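; Quest gameplay ability that credits a player for bouncing a set number of times while holding a chicken.
; On event activation it resets the bounce count and starts tasks watching for tire and web bounce tag additions and a landing event.
; At max bounces it checks the held object's chicken tag, sends the quest stat if present, and ends either way.
; Landing also ends the ability, which resets the count and stops the tasks.
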(event Ability|EventActivateAbilityFromEvent
  (params (EventData :type Gameplay%20Event%20Data%20Structure%20%28by%20ref%29))
  (bind (_eventtag _instigator _target _optionalobject _optionalobject2 _contexthandle _instigatortags _targettags) (Utilities|Struct|BreakGameplayEventData EventData))
  (|CallParentFunction EventData)
  (Variables|Tags|SetTargetTags _targettags)
  ;[1 0.2402 0.03963] Initialize bounce tracking...
  (Variables|Default|SetBounceCount 0)
  (bind _asynctaskproxy (Ability|Tasks|WaitGameplayEvent (Variables|Tags|GetTrackedLandingTag) 0 true)
    (:then
      (Variables|Tasks|SetLandedTagTrackTask _asynctaskproxy)
      (bind _asynctaskproxy_2 (Ability|Tasks|WaitGameplayTagAdd (Variables|Tags|GetTrackedTireBounceTag))
        (:then
          (Variables|Tasks|SetTireTagTrackTask _asynctaskproxy_2)
          (bind _asynctaskproxy_3 (Ability|Tasks|WaitGameplayTagAdd (Variables|Tags|GetTrackedWebBounceTag))
            (:then
              (Variables|Tasks|SetWebTagTrackTask _asynctaskproxy_3))
            (:Added
              _)))
        (:Added
          _))
      (bind _result (Math|Integer|IncrementInt (Variables|Default|GetBounceCount))
        (:%20%20
          (bind _output_get (Variables|Default|SetBounceCount _result))
          ;[1 1 1] If not max bounce count just keep counting bounces until the max is reached then check for chicken
          (if (>= _output_get (Variables|Default|GetMaxBounces))
            (bind _bholding (CallFunction|IsHoldingChicken))
            (if _bholding
              ;[0.03963 0.9403 1] Max bounce count reached and holding chicken. Send quest update and then end ability to preform cleanup
              (Quest|SendComplexCustomStatEvent (Variables|GAVars|Data|GetOwningPlayerQuestManager) 0 (Variables|Tags|GetStatEventSourceTags) (Variables|Tags|GetTargetTags) 1)
              (Ability|EndAbility)
              (else
                (Ability|EndAbility)))))))
    (:EventReceived
      ;[1 0.09399 0.7665] Target landed...
      (Ability|EndAbility))))

(event Ability|EventOnEndAbility
  (params (bWasCancelled :type Boolean))
  (|CallParentFunction bWasCancelled)
  (Variables|Default|SetBounceCount 0)
  (GameplayTasks|EndTask (Variables|Tasks|GetLandedTagTrackTask))
  (GameplayTasks|EndTask (Variables|Tasks|GetWebTagTrackTask))
  (GameplayTasks|EndTask (Variables|Tasks|GetTireTagTrackTask)))

(fn IsHoldingChicken
  (if (Class|FortPawn|GetIsHoldingObject (Variables|GAVars|Data|GetOwningPlayerPawn))
    (bind _returnvalue (Fortnite|HeldObjects|GetHeldObjectAttachedtoPlayer (Variables|GAVars|Data|GetOwningPlayerPawn)))
    (Utilities|IsValid _returnvalue
      (:Is%20Valid
        (bind _bdidactorhavetags (Class|AthenaQuestFunctionLibrary|AthenaCheckActorHasTags _returnvalue (GameplayTags|MakeGameplayTagContainerfromTag (Variables|Tags|GetChickenTag)) true))
        (return _bdidactorhavetags))
      (:Is%20Not%20Valid
        (return false)))
    (else
      (return false))))
